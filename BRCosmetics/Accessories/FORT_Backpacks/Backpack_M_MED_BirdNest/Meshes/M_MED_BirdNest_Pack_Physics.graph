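# Physics asset for the M_MED_BirdNest backpack cosmetic, covering a small figure's dangling arms, legs and ears plus the pack itself.
# Three-segment arm and leg chains and single-body ears constrain back to body_base_c, using capsules with spheres at the arm ends.
# Angular limits are widest at the shoulders, with mid joints swinging on one axis like hinges.
# The large pack capsule has no constraints.
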PhysicsAsset /BRCosmetics/Accessories/FORT_Backpacks/Backpack_M_MED_BirdNest/Meshes/M_MED_BirdNest_Pack_Physics
bodies {
 body_base_c [{shapeType=Capsule, length=7.75, radius=9.25, shapeName=body_base_c_capsule, center=<0 -1.076 2.1388>}]
 dyn_arm_end_base_l [{shapeType=Sphere, radius=5.5, shapeName=dyn_arm_end_base_l_sphere, center=<-5.3561 -0.4985 -0.5929>}]
 dyn_arm_end_base_r [{shapeType=Sphere, radius=5.5, shapeName=dyn_arm_end_base_l_sphere, center=<5.3561 0.4985 0.5929>}]
 dyn_arm_mid_base_l [{
   shapeType=Capsule
   length=3.25
   radius=1
   shapeName=dyn_arm_mid_base_l_capsule
   center=<-1.6759 0 -0.2955>
   rotation=<rot -80 0.0001 -0.0002>
  }]
 dyn_arm_mid_base_r [{shapeType=Capsule, length=3.25, radius=1, shapeName=dyn_arm_mid_base_l_capsule, center=<1.6759 0 0.2955>, rotation=<rot -80 0.0001 179.9998>}]
 dyn_arm_start_base_l [{shapeType=Capsule, length=2.5, radius=0.75, shapeName=dyn_arm_start_base_l_capsule, center=<-1.3013 0.072 0>, rotation=<rot -90 0 -0.0002>}]
 dyn_arm_start_base_r [{shapeType=Capsule, length=2.5, radius=0.75, shapeName=dyn_arm_start_base_l_capsule, center=<1.3013 -0.072 0>, rotation=<rot -90 0 179.9998>}]
 dyn_ear_base_l [{shapeType=Capsule, length=1.5, radius=1.5, shapeName=dyn_ear_base_l_capsule, center=<0 0 -1.0623>}]
 dyn_ear_base_r [{shapeType=Capsule, length=1.5, radius=1.5, shapeName=dyn_ear_base_l_capsule, center=<0 0 1.0623>, rotation=<rot 0 0 -180>}]
 dyn_leg_end_base_l [{shapeType=Capsule, length=2.75, radius=2, shapeName=dyn_leg_end_base_l_capsule, center=<-3.8083 0 0.6715>, rotation=<rot -80 180 -180>}]
 dyn_leg_end_base_r [{shapeType=Capsule, length=2.75, radius=2, shapeName=dyn_leg_end_base_l_capsule, center=<3.8083 0 -0.6715>, rotation=<rot -80 180 0>}]
 dyn_leg_mid_base_l [{
   shapeType=Capsule
   length=3
   radius=2.25
   shapeName=dyn_leg_mid_base_l_capsule
   center=<-4.0344 0 0.151>
   rotation=<rot 90 -12.0048 -12.0049>
  }]
 dyn_leg_mid_base_r [{shapeType=Capsule, length=3, radius=2.25, shapeName=dyn_leg_mid_base_l_capsule, center=<4.0344 0 -0.151>, rotation=<rot 90 -3.3723 176.6276>}]
 dyn_leg_start_base_l [{
   shapeType=Capsule
   length=3
   radius=1.25
   shapeName=dyn_leg_start_base_l_capsule
   center=<-1.8044 0 0>
   rotation=<rot -90 -174.9489 174.9489>
  }]
 dyn_leg_start_base_r [{shapeType=Capsule, length=3, radius=1.25, shapeName=dyn_leg_start_base_l_capsule, center=<1.8044 0 0>, rotation=<rot -90 -179.0505 -0.9494>}]
 pack [{shapeType=Capsule, length=33.75, radius=25.5, shapeName=pack_capsule, center=<0 21.064 -0.1579>, rotation=<rot 0 0 5>}]
}
constraints [
 {bone1Name=dyn_leg_start_base_l, bone2Name=body_base_c, swing1LimitDegrees=25, swing2LimitDegrees=25, twistLimitDegrees=5}
 {bone1Name=dyn_leg_mid_base_l, bone2Name=dyn_leg_start_base_l, swing1LimitDegrees=0, swing2LimitDegrees=25, twistLimitDegrees=0}
 {bone1Name=dyn_leg_end_base_l, bone2Name=dyn_leg_mid_base_l, swing1LimitDegrees=15, swing2LimitDegrees=20, twistLimitDegrees=0}
 {bone1Name=dyn_ear_base_l, bone2Name=body_base_c, swing1LimitDegrees=20, swing2LimitDegrees=5, twistLimitDegrees=35}
 {bone1Name=dyn_leg_end_base_r, bone2Name=dyn_leg_mid_base_r, swing1LimitDegrees=10, swing2LimitDegrees=25, twistLimitDegrees=0}
 {bone1Name=dyn_leg_start_base_r, bone2Name=body_base_c, swing1LimitDegrees=30, swing2LimitDegrees=30, twistLimitDegrees=5}
 {bone1Name=dyn_arm_mid_base_r, bone2Name=dyn_arm_start_base_r, swing1LimitDegrees=25, swing2LimitDegrees=0, twistLimitDegrees=0}
 {bone1Name=dyn_arm_end_base_l, bone2Name=dyn_arm_mid_base_l, swing1LimitDegrees=10, swing2LimitDegrees=3, twistLimitDegrees=5}
 {bone1Name=dyn_arm_start_base_l, bone2Name=body_base_c, swing1LimitDegrees=85, swing2LimitDegrees=35, twistLimitDegrees=5}
 {bone1Name=dyn_arm_mid_base_l, bone2Name=dyn_arm_start_base_l, swing1LimitDegrees=30, swing2LimitDegrees=0, twistLimitDegrees=0}
 {bone1Name=dyn_arm_start_base_r, bone2Name=body_base_c, swing1LimitDegrees=65, swing2LimitDegrees=35, twistLimitDegrees=5}
 {bone1Name=dyn_arm_end_base_r, bone2Name=dyn_arm_mid_base_r, swing1LimitDegrees=10, swing2LimitDegrees=3, twistLimitDegrees=5}
 {bone1Name=dyn_ear_base_r, bone2Name=body_base_c, swing1LimitDegrees=5, swing2LimitDegrees=20, twistLimitDegrees=35}
 {bone1Name=dyn_leg_mid_base_r, bone2Name=dyn_leg_start_base_r, swing1LimitDegrees=0, swing2LimitDegrees=25, twistLimitDegrees=0}
]
