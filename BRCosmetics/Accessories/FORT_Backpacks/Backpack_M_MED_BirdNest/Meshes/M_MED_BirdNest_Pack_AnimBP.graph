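# Anim Blueprint for the Bird Nest backpack, adding physics motion and swapping visible material elements when the player fires.
# The AnimGraph offsets the root bone, then runs a RigidBody simulation.
# Outside the front end, with Melee false (set from the player's weapon animation type), firing toggles two Hide Element parameters, restored after a three-second retriggerable delay.
# Each update also stores a gravity override from per-movement-state parameters via a deprecated function.
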AnimBlueprint /BRCosmetics/Accessories/FORT_Backpacks/Backpack_M_MED_BirdNest/Meshes/M_MED_BirdNest_Pack_AnimBP
target_skeleton=&/M_MED_BirdNest_Pack_Skeleton
graphs [
 AnimGraph {nodes [
   n1 = "Transform (Modify) Bone - Bone: root"(ComponentPose=Local To Component(LocalPose=Input Pose()), Translation="0,1,0", Rotation="0,0,-6") {
    rotationMode=BMM_Additive
    rotationSpace=BCS_BoneSpace
    translationMode=BMM_Additive
    translationSpace=BCS_BoneSpace
    alphaBoolBlend {blendOption=Linear}
    boneToModify {boneName=root}
    rotation {roll=-6}
    translation {y=1}
   }
   n2 = RigidBody!(ComponentPose=@n1, Alpha=0.8, OverrideWorldGravity:Vector="0,0,0", ExternalForce) -> (Pose) {
    bClampLinearTranslationLimitToRefPose=true
    bOverrideWorldGravity=true
    overridePhysicsAsset=&/M_MED_BirdNest_Pack_Physics
    alphaBoolBlend {blendOption=Linear}
    componentAppliedLinearAccClamp=<1000 1000 1000>
    componentLinearAccScale=<0.2 0.2 0.1>
    componentLinearVelScale=<0.6 0.6 1>
    simSpaceSettings {worldAlpha=0.2}
   }
   Output Pose(Result=Component To Local(ComponentPose=@n2)) {layerGroup=None}
  ]}
 EventGraph {nodes [
   Event BlueprintUpdateAnimation()
   > n1 = Sequence!(execute) -> (then_0, then_1, then_2:Exec)
   n2 = Branch(execute=@n1.then_1, Condition=Get Melee() -> (Melee:bool))
   n3 = Get Dynamic Material Instance(execute=@n2.else, SkeletalMeshComponent=GetPartSkeletalMeshComponent())
   > n4 = Branch(Condition=Get bIsInFrontEnd())
   n5 = Branch(execute=@n4.else, Condition=Get bRecentlyFired() -> (bRecentlyFired:bool))
   > n6 = Sequence!(execute) -> (then_0, then_1)
   Do Once(execute=@n5.else, Reset=@n6.then_1)
   > RetriggerableDelay(Duration=3)
   > SetScalarParameterValue(self=@n3, ParameterName=Hide Element 05, Value=1)
   > SetScalarParameterValue(self=@n3, ParameterName=Hide Element 02, Value=0)
   SetScalarParameterValue(execute=@n6.then_0, self=@n3, ParameterName=Hide Element 02, Value=1)
   > SetScalarParameterValue(self=@n3, ParameterName=Hide Element 05, Value=0)
   n7 = Gravity Override State Function DEPRECATED(
    execute=@n1.then_2
    CharacterPartAnimInstance=Self-Reference()
    GameTimeInSeconds=GetGameTimeInSeconds()
    GravityOverrideParams=Get Gravity Override Params()
   )
   > Set Gravity Override(Gravity Override=@n7)
   n8 = Cast To FortPlayerAnimInstance(execute=@n1.then_0, Object=GetAnimInstance(self=GetOwnerSkeletalMeshComponent()))
   > n9 = Switch on EFortWeaponCoreAnimation!(execute, Selection:EFortWeaponCoreAnimation Enum=Get WeaponCoreAnim(self=@n8)) -> (
    Melee:Exec
    Pistol:Exec
    Shotgun:Exec
    PaperBlueprint:Exec
    Rifle:Exec
    MeleeOneHand:Exec
    MachinePistol:Exec
    RocketLauncher:Exec
    GrenadeLauncher:Exec
    GoingCommando:Exec
    AssaultRifle:Exec
    TacticalShotgun:Exec
    SniperRifle:Exec
    TrapPlacement:Exec
    ShoulderLauncher:Exec
    AbilityDecoTool:Exec
    Crossbow:Exec
    C4:Exec
    RemoteControl:Exec
    DualWield:Exec
    AR_BullPup:Exec
    AR_ForwardGrip:Exec
    MedPackPaddles:Exec
    SMG_P90:Exec
    AR_DrumGun:Exec
    Consumable_Small:Exec
    Consumable_Large:Exec
    Balloon:Exec
    MountedTurret:Exec
    CreativeTool:Exec
    ExplosiveBow:Exec
    AshtonIndigo:Exec
    AshtonChicago:Exec
    MeleeDualWield:Exec
    Unarmed:Exec
   )
   Set Melee(
    execute=[
     @n9.AR_DrumGun
     @n9.AssaultRifle
     @n9.GrenadeLauncher
     @n9.PaperBlueprint
     @n9.Pistol
     @n9.Rifle
     @n9.RocketLauncher
     @n9.SMG_P90
     @n9.Shotgun
     @n9.SniperRifle
     @n9.TacticalShotgun
    ]
    Melee:bool=false
   ) -> (Output_Get:bool)
   Set Melee(execute=[@n9.DualWield, @n9.Melee, @n9.MeleeDualWield, @n9.MeleeOneHand], Melee=true)
   Event BlueprintInitializeAnimation()
   > n10 = Do Once()
   SetScalarParameterValue(execute=[@n10.Completed, @n11.Completed], self=@n3, ParameterName=Hide Element 02, Value=0)
   > SetScalarParameterValue(self=@n3, ParameterName=Hide Element 05, Value=1)
   Event OnCharacterCustomizationLoaded()
   > n11 = Do Once()
   TryGetPawnOwner()
  ]}
]
variables [
 Gravity Override {type=Vector}
 Gravity Override Params {
  type=GravityOverrideParamsStruct
  default {
   onGround_Moving {gravityOverride=<0 0 -100>}
   rail_Grinding {gravityOverride=<0 -500 300>}
   skydive_DivingDown {gravityOverride=<0 -980 0>, windFrequency=<1.5 7 8>}
   skydive_FloatingDown {gravityOverride=<0 0 330>, windOffset=<0 -300 300>}
   skydive_GoingUp {gravityOverride=<0 0 330>, windOffset=<0 -400 400>}
   skydive_Parachuting {gravityOverride=<0 0 0>, windAmplitude=<20 -10 40>}
   tactical_Sprint {gravityOverride=<0 50 40>}
  }
 }
 Melee {type=boolean}
]
